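# Physics asset for the skeletal mesh of the Pickaxe_36 melee weapon cosmetic.
# It holds a single box body on the Melee_TwoHanded_Handle bone, offset and rotated to cover the tall, thin pickaxe shape.
# It has no constraints, so the weapon simulates and collides as one rigid piece.
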PhysicsAsset /BRCosmetics/Weapons/FORT_Melee/Pickaxe_36/Meshes/SK_Pickaxe_36_PhysicsAsset
bodies {Melee_TwoHanded_Handle [{shapeType=Box, extentX=49.2361, extentY=8.5672, extentZ=170.1426, center=<-0.0769 -1.8608 44.3132>, rotation=<rot 0.276 -90 0.0973>}]}
constraints []
